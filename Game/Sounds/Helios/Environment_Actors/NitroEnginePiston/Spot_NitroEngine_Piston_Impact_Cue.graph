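# Sound cue for the Helios NitroEngine piston impact, randomly picking one of three pitch-modulated impact waves with its own logarithmic falloff.
# A second branch randomizes twelve metal spark bursts into indoor and lower-pitched outdoor sound-class layers, with custom distance curves and reverb sends.
# That spark branch is mixed at zero volume, leaving only the impact audible, and its quality-level node routes every tier to the same mixer.
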SoundCue /Game/Sounds/Helios/Environment_Actors/NitroEnginePiston/Spot_NitroEngine_Piston_Impact_Cue
root=Mixer_2
nodes [
 {
  id=Attenuation_1
  children [Modulator_2]
  properties {bOverrideAttenuation=true, attenuationOverrides {distanceAlgorithm=Logarithmic, falloffDistance=2750, attenuationShapeExtents {x=450}}}
 }
 {
  id=Attenuation_3
  children [SoundClass_1]
  properties {
   bOverrideAttenuation=true
   attenuationOverrides {
    bEnableListenerFocus=true
    bEnableSubmixSends=true
    distanceAlgorithm=Custom
    falloffDistance=8000
    nonFocusVolumeAttenuation=0.2
    attenuationShapeExtents {x=0}
    customAttenuationCurve {editorCurveData {keys [
       {arriveTangent=3.2291, interpMode=RCIM_Cubic, leaveTangent=3.2292, tangentMode=RCTM_User, time=0, value=0}
       {arriveTangent=-0.0168, interpMode=RCIM_Cubic, leaveTangent=-0.0168, tangentMode=RCTM_User, time=0.2, value=1.0007}
       {arriveTangent=-1.2508, interpMode=RCIM_Cubic, leaveTangent=-1.2508, time=0.5439, value=0.3007}
       {arriveTangent=-1.6591, interpMode=RCIM_Cubic, leaveTangent=-1.6591, tangentMode=RCTM_User, time=1, value=0}
      ]}}
    submixSendSettings [{
      bEnableHPFCutoff=false
      bEnableLPFCutoff=false
      disableManualSendClamp=false
      hPFCutoff=20
      lPFCutoff=20000
      maxSendDistance=6000
      maxSendLevel=1
      minSendDistance=400
      minSendLevel=1
      sendLevel=1
      sendLevelControlMethod=Linear
      soundSubmix=&Fortnite_Reverb_Default
     }]
   }
  }
 }
 {
  id=Attenuation_4
  children [SoundClass_0]
  properties {
   bOverrideAttenuation=true
   attenuationOverrides {
    bEnableListenerFocus=true
    bEnableSubmixSends=true
    distanceAlgorithm=Custom
    falloffDistance=4000
    nonFocusVolumeAttenuation=0.75
    stereoSpread=500
    attenuationShapeExtents {x=0}
    customAttenuationCurve {editorCurveData {keys [
       {arriveTangent=-0.0131, interpMode=RCIM_Cubic, leaveTangent=-0.0131, tangentMode=RCTM_User, time=0.4, value=1.0124}
       {arriveTangent=-0.0228, interpMode=RCIM_Cubic, leaveTangent=-0.0228, tangentMode=RCTM_User, time=1, value=0}
      ]}}
    submixSendSettings [{
      bEnableHPFCutoff=false
      bEnableLPFCutoff=false
      disableManualSendClamp=false
      hPFCutoff=20
      lPFCutoff=20000
      maxSendDistance=6000
      maxSendLevel=1
      minSendDistance=400
      minSendLevel=0.4
      sendLevel=1
      sendLevelControlMethod=Linear
      soundSubmix=&Fortnite_Reverb_Default
      customSendLevelCurve {editorCurveData {keys [<at 0.2 0.4>,<at 1 1>]}}
     }]
   }
  }
 }
 {id=Mixer_0, children [Attenuation_4, Attenuation_3], properties {inputVolume [1,0.2]}}
 {id=Mixer_2, children [Attenuation_1, QualityLevel_0], properties {inputVolume [1,0]}}
 {id=Modulator_0, children [Random_2], properties {pitchMin=0.8}}
 {id=Modulator_1, children [Random_2], properties {pitchMax=0.8, pitchMin=0.7}}
 {id=Modulator_2, children [Random_0], properties {pitchMax=1}}
 {id=QualityLevel_0, children [Mixer_0,Mixer_0,Mixer_0,Mixer_0]}
 {id=Random_0, children [WavePlayer_12, WavePlayer_1, WavePlayer_3], properties {weights [1,1,1]}}
 {
  id=Random_2
  children [
   WavePlayer_0
   WavePlayer_2
   WavePlayer_6
   WavePlayer_7
   WavePlayer_8
   WavePlayer_9
   WavePlayer_10
   WavePlayer_11
   WavePlayer_16
   WavePlayer_17
   WavePlayer_18
   WavePlayer_19
  ]
  properties {weights [1,1,1,1,1,1,1,1,1,1,1,1]}
 }
 {id=SoundClass_0, children [Modulator_0], properties {soundClassOverride=/Game/Sounds/FortSoundClassesAndModes/Fort_Ambient_OneShots_Indoors}}
 {
  id=SoundClass_1
  children [Modulator_1]
  properties {soundClassOverride=&Fort_Ambient_OneShots_Outdoors}
 }
 {id=WavePlayer_0, properties {soundWaveAssetPtr=/Game/Sounds/Helios/POI/NitroEngine/Spot/Spot_Spark_Burst_Metal_Int_01}}
 {id=WavePlayer_1, properties {soundWaveAssetPtr=&/NitroEngine_Piston_Impact_04}}
 {id=WavePlayer_10, properties {soundWaveAssetPtr=/Game/Sounds/Helios/POI/NitroEngine/Spot/Spot_Spark_Burst_Metal_Int_07}}
 {id=WavePlayer_11, properties {soundWaveAssetPtr=/Game/Sounds/Helios/POI/NitroEngine/Spot/Spot_Spark_Burst_Metal_Int_08}}
 {id=WavePlayer_12, properties {soundWaveAssetPtr=&/NitroEngine_Piston_Impact_01}}
 {id=WavePlayer_16, properties {soundWaveAssetPtr=/Game/Sounds/Helios/POI/NitroEngine/Spot/Spot_Spark_Burst_Metal_Int_09}}
 {id=WavePlayer_17, properties {soundWaveAssetPtr=/Game/Sounds/Helios/POI/NitroEngine/Spot/Spot_Spark_Burst_Metal_Int_10}}
 {id=WavePlayer_18, properties {soundWaveAssetPtr=/Game/Sounds/Helios/POI/NitroEngine/Spot/Spot_Spark_Burst_Metal_Int_11}}
 {id=WavePlayer_19, properties {soundWaveAssetPtr=/Game/Sounds/Helios/POI/NitroEngine/Spot/Spot_Spark_Burst_Metal_Int_12}}
 {id=WavePlayer_2, properties {soundWaveAssetPtr=/Game/Sounds/Helios/POI/NitroEngine/Spot/Spot_Spark_Burst_Metal_Int_02}}
 {id=WavePlayer_3, properties {soundWaveAssetPtr=&/NitroEngine_Piston_Impact_03}}
 {id=WavePlayer_6, properties {soundWaveAssetPtr=/Game/Sounds/Helios/POI/NitroEngine/Spot/Spot_Spark_Burst_Metal_Int_03}}
 {id=WavePlayer_7, properties {soundWaveAssetPtr=/Game/Sounds/Helios/POI/NitroEngine/Spot/Spot_Spark_Burst_Metal_Int_04}}
 {id=WavePlayer_8, properties {soundWaveAssetPtr=/Game/Sounds/Helios/POI/NitroEngine/Spot/Spot_Spark_Burst_Metal_Int_05}}
 {id=WavePlayer_9, properties {soundWaveAssetPtr=/Game/Sounds/Helios/POI/NitroEngine/Spot/Spot_Spark_Burst_Metal_Int_06}}
]
properties {
 bOverrideConcurrency=true
 maxDistance=8000
 soundClassObject=&Fort_SFX_InGame
 volumeMultiplier=0.126
 attenuationOverrides {
  bEnableListenerFocus=true
  distanceAlgorithm=Logarithmic
  falloffDistance=1200
  focusDistanceScale=0.9
  nonFocusVolumeAttenuation=0.4
  stereoSpread=700
  attenuationShapeExtents=<250 120 -60.9851>
  customAttenuationCurve {editorCurveData {keys [{arriveTangent=-2.3812, interpMode=RCIM_Cubic, leaveTangent=-2.3812, tangentMode=RCTM_User, time=0, value=1}, <at 1 0.01 RCIM_Cubic>]}}
 }
 concurrencyOverrides {maxCount=8}
}
