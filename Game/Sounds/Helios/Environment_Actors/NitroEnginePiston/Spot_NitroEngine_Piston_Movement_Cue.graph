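# Sound cue for the movement of a nitro engine piston environment actor in the Helios audio set, routed to the in-game SFX class.
# A random node picks between two equally weighted piston waves, which pass through a modulator for pitch and volume variation.
# The root attenuation node applies its own logarithmic falloff, while the cue halves volume and overrides concurrency to cap simultaneous instances.
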SoundCue /Game/Sounds/Helios/Environment_Actors/NitroEnginePiston/Spot_NitroEngine_Piston_Movement_Cue
root=Attenuation_1
nodes [
 {
  id=Attenuation_1
  children [Modulator_2]
  properties {bOverrideAttenuation=true, attenuationOverrides {distanceAlgorithm=Logarithmic, falloffDistance=2750, attenuationShapeExtents {x=450}}}
 }
 {id=Modulator_2, children [Random_0], properties {pitchMax=1}}
 {id=Random_0, children [WavePlayer_0,WavePlayer_1], properties {weights [1,1]}}
 {id=WavePlayer_0, properties {soundWaveAssetPtr=&/NitroEngine_Piston_Movement_01}}
 {id=WavePlayer_1, properties {soundWaveAssetPtr=&/NitroEngine_Piston_Movement_03}}
]
properties {
 bOverrideConcurrency=true
 maxDistance=3200
 soundClassObject=&Fort_SFX_InGame
 volumeMultiplier=0.5
 attenuationOverrides {
  bEnableListenerFocus=true
  distanceAlgorithm=Logarithmic
  falloffDistance=1200
  focusDistanceScale=0.9
  nonFocusVolumeAttenuation=0.4
  stereoSpread=700
  attenuationShapeExtents=<250 120 -60.9851>
  customAttenuationCurve {editorCurveData {keys [{arriveTangent=-2.3812, interpMode=RCIM_Cubic, leaveTangent=-2.3812, tangentMode=RCTM_User, time=0, value=1}, <at 1 0.01 RCIM_Cubic>]}}
 }
 concurrencyOverrides {maxCount=4}
}
